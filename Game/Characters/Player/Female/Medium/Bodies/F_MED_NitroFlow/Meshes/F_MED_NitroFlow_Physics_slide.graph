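# Physics asset for the F_MED_NitroFlow female player body, a "slide" variant that simulates the skirt against the lower body.
# Capsules cover the pelvis, thighs, calves and six dyn_skirt bone chains (front, side and back on each side), with a noticeably fatter right calf capsule.
# Angular-limited constraints link each chain, with back and side roots tied to the pelvis.
# Front chains have no pelvis constraint and are linked only from segment 1 down.
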PhysicsAsset /Game/Characters/Player/Female/Medium/Bodies/F_MED_NitroFlow/Meshes/F_MED_NitroFlow_Physics_slide
bodies {
 calf_l [{shapeType=Capsule, length=45.0106, radius=5.1362, center=<-27.2829 -0.9102 -0.3922>, rotation=<rot -0.4008 89.0505 92.2757>}]
 calf_r [{shapeType=Capsule, length=36.2655, radius=13.8813, center=<27.2046 3.2484 -5.714>, rotation=<rot -0.4008 89.0504 -82.3166>}]
 dyn_skirt_back_l_0 [{
   shapeType=Capsule
   length=6.1404
   radius=7.2661
   shapeName=dyn_skirt_back_l_0_capsule
   center=<13.8587 -3.2379 -0.2121>
   rotation=<rot 7.8327 -89.6035 3.168>
  }]
 dyn_skirt_back_l_1 [{
   shapeType=Capsule
   length=8.5621
   radius=6.5526
   shapeName=dyn_skirt_back_l_1_capsule
   center=<18.3039 -5.8068 1.7666>
   rotation=<rot 7.0727 -89.8354 1.5857>
  }]
 dyn_skirt_back_l_2 [{
   shapeType=Capsule
   length=24.3216
   radius=8.8052
   shapeName=dyn_skirt_back_l_2_capsule
   center=<13.3586 -9.3103 2.1938>
   rotation=<rot 4.2376 -89.0529 -0.9879>
  }]
 dyn_skirt_back_l_3 [{
   shapeType=Capsule
   length=24.572
   radius=9.2603
   shapeName=dyn_skirt_back_l_3_capsule
   center=<12.5467 -11.4001 2.0442>
   rotation=<rot -4.4202 85.4412 -1.7991>
  }]
 dyn_skirt_back_l_4 [{
   shapeType=Capsule
   length=33.3926
   radius=10.1073
   shapeName=dyn_skirt_back_l_4_capsule
   center=<13.3468 -14.2786 2.4122>
   rotation=<rot 3.3154 -90.915 0.3279>
  }]
 dyn_skirt_back_r_0 [{
   shapeType=Capsule
   length=4.8551
   radius=7.2661
   shapeName=dyn_skirt_back_l_0_capsule
   center=<-13.9328 3.2082 -0.9559>
   rotation=<rot 7.6843 -89.6955 -177.5806>
  }]
 dyn_skirt_back_r_1 [{
   shapeType=Capsule
   length=7.2717
   radius=7.2991
   shapeName=dyn_skirt_back_l_1_capsule
   center=<-18.437 5.2545 -2.3185>
   rotation=<rot 7.1206 -89.801 -177.9975>
  }]
 dyn_skirt_back_r_2 [{
   shapeType=Capsule
   length=21.5934
   radius=9.5517
   shapeName=dyn_skirt_back_l_2_capsule
   center=<-13.4452 9.6321 -2.9005>
   rotation=<rot 4.2553 -89.0372 179.243>
  }]
 dyn_skirt_back_r_3 [{
   shapeType=Capsule
   length=24.1665
   radius=10.0068
   shapeName=dyn_skirt_back_l_3_capsule
   center=<-10.5518 11.4762 -2.07>
   rotation=<rot -4.5288 85.4945 177.4559>
  }]
 dyn_skirt_back_r_4 [{
   shapeType=Capsule
   length=31.2976
   radius=10.1316
   shapeName=dyn_skirt_back_l_4_capsule
   center=<-13.7531 15.5098 -2.4635>
   rotation=<rot 3.515 -90.7842 -177.2056>
  }]
 dyn_skirt_front_l_0 [{
   shapeType=Capsule
   length=20.0938
   radius=4.4688
   shapeName=dyn_skirt_front_l_0_capsule
   center=<13.2556 5.775 -10.5213>
   rotation=<rot 20.228 -89.6602 -173.448>
  }]
 dyn_skirt_front_l_1 [{
   shapeType=Capsule
   length=3.573
   radius=1.7316
   shapeName=dyn_skirt_front_l_1_capsule
   center=<18.1061 4.6837 -3.656>
   rotation=<rot 84.5033 0.0043 -0.0002>
  }]
 dyn_skirt_front_l_2 [{
   shapeType=Capsule
   length=10.0104
   radius=8.4934
   shapeName=dyn_skirt_front_l_2_capsule
   center=<13.5637 2.2033 12.1187>
   rotation=<rot -0.1446 0.2784 -65.0888>
  }]
 dyn_skirt_front_l_3 [{
   shapeType=Capsule
   length=10.7826
   radius=8.5704
   shapeName=dyn_skirt_front_l_3_capsule
   center=<12.1566 1.5173 12.3094>
   rotation=<rot 11.9856 174.8296 -101.4594>
  }]
 dyn_skirt_front_l_4 [{
   shapeType=Capsule
   length=12.1371
   radius=8.7725
   shapeName=dyn_skirt_front_l_4_capsule
   center=<15.1321 3.5352 20.4191>
   rotation=<rot 2.527 178.0626 -98.5409>
  }]
 dyn_skirt_front_r_1 [{
   shapeType=Capsule
   length=3.6041
   radius=1.7124
   shapeName=dyn_skirt_front_l_1_capsule
   center=<-18.1061 -4.6837 3.656>
   rotation=<rot 84.5033 0.0043 179.9998>
  }]
 dyn_skirt_front_r_2 [{
   shapeType=Capsule
   length=10.6513
   radius=7.8526
   shapeName=dyn_skirt_front_l_2_capsule
   center=<-13.6476 -2.4354 -14.5434>
   rotation=<rot -0.1446 0.2784 114.9112>
  }]
 dyn_skirt_front_r_3 [{
   shapeType=Capsule
   length=11.4234
   radius=7.9295
   shapeName=dyn_skirt_front_l_3_capsule
   center=<-13.4765 -2.5852 -18.2335>
   rotation=<rot 11.9856 174.8296 78.5406>
  }]
 dyn_skirt_front_r_4 [{
   shapeType=Capsule
   length=12.7779
   radius=8.1316
   shapeName=dyn_skirt_front_l_4_capsule
   center=<-15.3276 -4.0097 -24.5183>
   rotation=<rot 2.527 178.0626 81.4591>
  }]
 dyn_skirt_l_0 [{
   shapeType=Capsule
   length=11.0474
   radius=7.1975
   shapeName=dyn_skirt_l_0_capsule
   center=<15.0644 -0.732 6.2102>
   rotation=<rot 1.4797 0.9468 -95.6805>
  }]
 dyn_skirt_l_1 [{
   shapeType=Capsule
   length=8.831
   radius=6.955
   shapeName=dyn_skirt_l_1_capsule
   center=<19.0191 -3.1617 5.7694>
   rotation=<rot -8.1591 1.2235 -95.7802>
  }]
 dyn_skirt_l_2 [{
   shapeType=Capsule
   length=23.0133
   radius=8.83
   shapeName=dyn_skirt_l_2_capsule
   center=<14.3453 -4.1035 4.9272>
   rotation=<rot -0.8906 -0.2662 -96.25>
  }]
 dyn_skirt_l_3 [{
   shapeType=Capsule
   length=25.5642
   radius=8.83
   shapeName=dyn_skirt_l_3_capsule
   center=<12.3616 -3.2737 4.3076>
   rotation=<rot 1.2084 0.2283 -97.3246>
  }]
 dyn_skirt_l_4 [{
   shapeType=Capsule
   length=33.045
   radius=9.3398
   shapeName=dyn_skirt_l_4_capsule
   center=<15.4766 -3.6798 10.7874>
   rotation=<rot 2.9768 -179.6535 -83.7859>
  }]
 dyn_skirt_r_0 [{
   shapeType=Capsule
   length=7.8996
   radius=7.6814
   shapeName=dyn_skirt_l_0_capsule
   center=<-12.3864 0.1869 -5.813>
   rotation=<rot 1.3846 -0.3768 86.25>
  }]
 dyn_skirt_r_1 [{
   shapeType=Capsule
   length=6.9396
   radius=6.842
   shapeName=dyn_skirt_l_1_capsule
   center=<-18.9477 1.047 -5.2003>
   rotation=<rot -8.2366 -0.0767 86.4918>
  }]
 dyn_skirt_r_2 [{
   shapeType=Capsule
   length=25.649
   radius=9.4007
   shapeName=dyn_skirt_l_2_capsule
   center=<-14.4967 2.9509 -6.8757>
   rotation=<rot -0.9625 -1.963 85.8559>
  }]
 dyn_skirt_r_3 [{
   shapeType=Capsule
   length=30.5035
   radius=7.8168
   shapeName=dyn_skirt_l_3_capsule
   center=<-12.0513 2.9906 -11.6842>
   rotation=<rot 1.1498 -0.6443 83.9433>
  }]
 dyn_skirt_r_4 [{
   shapeType=Capsule
   length=31.5346
   radius=9.7991
   shapeName=dyn_skirt_l_4_capsule
   center=<-15.7258 7.3864 -17.1437>
   rotation=<rot 3.0163 179.2899 95.7547>
  }]
 pelvis [{shapeType=Capsule, length=12.107, radius=8.9898, center=<-4.8633 -2.4439 -0.0582>, rotation=<rot -0.6044 -90.0116 0.9215>}]
 thigh_l [{shapeType=Capsule, length=26.3458, radius=6.5863, center=<-23.8457 -0.751 1.0569>, rotation=<rot 0.0001 -86.7873 90.3892>}]
 thigh_r [{shapeType=Capsule, length=26.3458, radius=6.5863, center=<23.8457 0.751 -1.0569>, rotation=<rot 0.0001 -86.7873 -89.6108>}]
}
constraints [
 {bone1Name=thigh_r, bone2Name=pelvis, swing1LimitDegrees=45, swing2LimitDegrees=45, twistLimitDegrees=45}
 {bone1Name=thigh_l, bone2Name=pelvis, swing1LimitDegrees=45, swing2LimitDegrees=45, twistLimitDegrees=45}
 {bone1Name=calf_l, bone2Name=thigh_l, swing1LimitDegrees=45, swing2LimitDegrees=45, twistLimitDegrees=45}
 {bone1Name=dyn_skirt_back_l_1, bone2Name=dyn_skirt_back_l_0, swing1LimitDegrees=35, swing2LimitDegrees=10, twistLimitDegrees=2}
 {bone1Name=dyn_skirt_back_l_2, bone2Name=dyn_skirt_back_l_1, swing1LimitDegrees=20, swing2LimitDegrees=10, twistLimitDegrees=5}
 {bone1Name=dyn_skirt_back_l_3, bone2Name=dyn_skirt_back_l_2, swing1LimitDegrees=10, swing2LimitDegrees=5, twistLimitDegrees=1}
 {bone1Name=dyn_skirt_back_l_4, bone2Name=dyn_skirt_back_l_3, swing1LimitDegrees=10, swing2LimitDegrees=5, twistLimitDegrees=1}
 {bone1Name=dyn_skirt_l_1, bone2Name=dyn_skirt_l_0, swing1LimitDegrees=15, swing2LimitDegrees=15, twistLimitDegrees=2}
 {bone1Name=dyn_skirt_l_2, bone2Name=dyn_skirt_l_1, swing1LimitDegrees=15, swing2LimitDegrees=15, twistLimitDegrees=5}
 {bone1Name=dyn_skirt_l_3, bone2Name=dyn_skirt_l_2, swing1LimitDegrees=15, swing2LimitDegrees=10, twistLimitDegrees=1}
 {bone1Name=dyn_skirt_l_4, bone2Name=dyn_skirt_l_3, swing1LimitDegrees=20, swing2LimitDegrees=10, twistLimitDegrees=1}
 {bone1Name=dyn_skirt_r_1, bone2Name=dyn_skirt_r_0, swing1LimitDegrees=20, swing2LimitDegrees=20, twistLimitDegrees=2}
 {bone1Name=dyn_skirt_r_2, bone2Name=dyn_skirt_r_1, swing1LimitDegrees=10, swing2LimitDegrees=20, twistLimitDegrees=5}
 {bone1Name=dyn_skirt_r_3, bone2Name=dyn_skirt_r_2, swing1LimitDegrees=10, swing2LimitDegrees=15, twistLimitDegrees=10}
 {bone1Name=dyn_skirt_r_4, bone2Name=dyn_skirt_r_3, swing1LimitDegrees=10, swing2LimitDegrees=10, twistLimitDegrees=1}
 {bone1Name=dyn_skirt_back_r_1, bone2Name=dyn_skirt_back_r_0, swing1LimitDegrees=30, swing2LimitDegrees=10, twistLimitDegrees=2}
 {bone1Name=dyn_skirt_back_r_2, bone2Name=dyn_skirt_back_r_1, swing1LimitDegrees=20, swing2LimitDegrees=10, twistLimitDegrees=5}
 {bone1Name=dyn_skirt_back_r_3, bone2Name=dyn_skirt_back_r_2, swing1LimitDegrees=10, swing2LimitDegrees=5, twistLimitDegrees=1}
 {bone1Name=dyn_skirt_back_r_4, bone2Name=dyn_skirt_back_r_3, swing1LimitDegrees=10, swing2LimitDegrees=5, twistLimitDegrees=1}
 {bone1Name=dyn_skirt_front_l_2, bone2Name=dyn_skirt_front_l_1, swing1LimitDegrees=30, swing2LimitDegrees=30, twistLimitDegrees=5}
 {bone1Name=dyn_skirt_front_l_3, bone2Name=dyn_skirt_front_l_2, swing1LimitDegrees=15, swing2LimitDegrees=35, twistLimitDegrees=10}
 {bone1Name=dyn_skirt_front_l_4, bone2Name=dyn_skirt_front_l_3, swing1LimitDegrees=20, swing2LimitDegrees=20, twistLimitDegrees=10}
 {bone1Name=dyn_skirt_front_r_2, bone2Name=dyn_skirt_front_r_1, swing1LimitDegrees=15, swing2LimitDegrees=15, twistLimitDegrees=5}
 {bone1Name=dyn_skirt_front_r_3, bone2Name=dyn_skirt_front_r_2, swing1LimitDegrees=10, swing2LimitDegrees=12, twistLimitDegrees=10}
 {bone1Name=dyn_skirt_front_r_4, bone2Name=dyn_skirt_front_r_3, swing1LimitDegrees=10, swing2LimitDegrees=12, twistLimitDegrees=10}
 {bone1Name=calf_r, bone2Name=thigh_r, swing1LimitDegrees=20, swing2LimitDegrees=10, twistLimitDegrees=1}
 {bone1Name=dyn_skirt_back_l_0, bone2Name=pelvis, swing1LimitDegrees=30, swing2LimitDegrees=5, twistLimitDegrees=2}
 {bone1Name=dyn_skirt_back_r_0, bone2Name=pelvis, swing1LimitDegrees=30, swing2LimitDegrees=5, twistLimitDegrees=2}
 {bone1Name=dyn_skirt_l_0, bone2Name=pelvis, swing1LimitDegrees=15, swing2LimitDegrees=10, twistLimitDegrees=2}
 {bone1Name=dyn_skirt_r_0, bone2Name=pelvis, swing1LimitDegrees=10, swing2LimitDegrees=10, twistLimitDegrees=2}
]
