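# Character-part AnimBP for the F_MED_NitroFlow body, adding secondary motion to its skirt and laces.
# The attached parent's pose passes through a Control Rig and upper-lace AnimDynamics, then is cached.
# Skirt RigidBody branches (thigh-curve-corrected normal, re-posed sliding) blend on bIsSliding with state-driven gravity, feeding leg-lace dynamics and emote-curve offsets.
# A skirt meta tag selects this chain or the cached pose; front-end curves adjust rig constraint weight.
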AnimBlueprint /Game/Characters/Player/Female/Medium/Bodies/F_MED_NitroFlow/Meshes/F_MED_NitroFlow_AnimBP
target_skeleton=&/F_MED_NitroFlow_Skeleton
graphs [
 AnimGraph {nodes [
   n1 = Get thigh_l_rotation_alpha() -> (thigh_l_rotation_alpha:double)
   n2 = "Transform (Modify) Bone - Bone: dyn_skirt_front_l_0"(
    ComponentPose=Local To Component(LocalPose=Use cached pose 'befor_skirt'())
    Alpha=@n1
    Translation="1.932,-2.490,2.320"
    Rotation="-10.232,7.169,12.482"
    Scale="1,1,1"
   ) {
    rotationMode=BMM_Additive
    rotationSpace=BCS_ParentBoneSpace
    scaleMode=BMM_Additive
    translationMode=BMM_Additive
    translationSpace=BCS_ParentBoneSpace
    alphaBoolBlend {blendOption=Linear}
    boneToModify {boneName=dyn_skirt_front_l_0}
    rotation=<rot -10.232 7.169 12.482>
    translation=<1.932 -2.49 2.32>
   }
   n3 = "Transform (Modify) Bone - Bone: dyn_skirt_front_l_1"(ComponentPose=@n2, Alpha=@n1, Translation="-3.637,1.265,6.525", Rotation="-0.962,-98.712,-66.549", Scale="1,1,1") {
    rotationMode=BMM_Additive
    rotationSpace=BCS_ParentBoneSpace
    scaleMode=BMM_Additive
    translationMode=BMM_Additive
    translationSpace=BCS_ParentBoneSpace
    alphaBoolBlend {blendOption=Linear}
    boneToModify {boneName=dyn_skirt_front_l_1}
    rotation=<rot -0.962 -98.712 -66.549>
    translation=<-3.637 1.265 6.525>
   }
   n4 = Get thigh_r_rotation_alpha() -> (thigh_r_rotation_alpha:double)
   n5 = "Transform (Modify) Bone - Bone: dyn_skirt_front_r_0"(ComponentPose=@n3, Alpha=@n4, Translation="-5.751,0.000,-0.000", Rotation="23.511,-8.328,16.949", Scale="1,1,1") {
    rotationMode=BMM_Additive
    rotationSpace=BCS_ParentBoneSpace
    scaleMode=BMM_Additive
    scaleSpace=BCS_ParentBoneSpace
    translationMode=BMM_Additive
    translationSpace=BCS_ParentBoneSpace
    alphaBoolBlend {blendOption=Linear}
    boneToModify {boneName=dyn_skirt_front_r_0}
    rotation=<rot 23.511 -8.328 16.949>
    translation {x=-5.751}
   }
   n6 = "Transform (Modify) Bone - Bone: dyn_skirt_front_r_1"(ComponentPose=@n5, Alpha=@n4, Translation="0,0,0", Rotation="-2.566,-64.177,-73.266", Scale="1,1,1") {
    rotationMode=BMM_Additive
    rotationSpace=BCS_ParentBoneSpace
    scaleMode=BMM_Additive
    translationMode=BMM_Additive
    translationSpace=BCS_ParentBoneSpace
    alphaBoolBlend {blendOption=Linear}
    boneToModify {boneName=dyn_skirt_front_r_1}
    rotation=<rot -2.5661 -64.1769 -73.2661>
   }
   n7 = RigidBody!(ComponentPose=@n6, Alpha, OverrideWorldGravity:Vector=Get Gravity Override()) -> (Pose) {
    bOverrideWorldGravity=true
    overridePhysicsAsset=&/F_MED_NitroFlow_Physics
    simulationSpace=BaseBoneSpace
    alphaBoolBlend {blendOption=Linear}
    baseBoneRef {boneName=pelvis}
    componentAppliedLinearAccClamp=<3000 3000 3000>
    componentLinearAccScale=<0.5 0.1 0.1>
    componentLinearVelScale=<0.1 0.1 0.1>
    overrideWorldGravity {z=-100}
    simSpaceSettings {maxAngularAcceleration=100, maxAngularVelocity=0.5, maxLinearAcceleration=100, maxLinearVelocity=100, velocityScaleZ=0, worldAlpha=1}
   }
   n8 = Component To Local(ComponentPose=@n7)
   n9 = "Transform (Modify) Bone - Bone: dyn_skirt_front_l_0_base"!(
    ComponentPose=Local To Component(LocalPose=Use cached pose 'befor_skirt'())
    bAlphaBoolEnabled:bool=True
    Translation="0,0,0"
    Rotation="50.926,32.228,66.645"
    Scale="1,1,1"
   ) -> (Pose) {
    alphaInputType=Bool
    rotationMode=BMM_Additive
    rotationSpace=BCS_ParentBoneSpace
    scaleMode=BMM_Additive
    translationMode=BMM_Additive
    translationSpace=BCS_ParentBoneSpace
    alphaBoolBlend {blendOption=Linear, blendOutTime=5}
    boneToModify {boneName=dyn_skirt_front_l_0_base}
    rotation=<rot 50.926 32.228 66.645>
   }
   n10 = "Transform (Modify) Bone - Bone: dyn_skirt_l_0_base"!(ComponentPose=@n9, bAlphaBoolEnabled:bool=True, Translation="0,0,0", Rotation="74.845,58.551,40.493", Scale="1,1,1") -> (Pose) {
    alphaInputType=Bool
    rotationMode=BMM_Additive
    rotationSpace=BCS_ParentBoneSpace
    scaleMode=BMM_Additive
    translationMode=BMM_Additive
    translationSpace=BCS_ParentBoneSpace
    alphaBoolBlend {blendOption=Linear, blendOutTime=5}
    boneToModify {boneName=dyn_skirt_l_0_base}
    rotation=<rot 74.845 58.551 40.493>
   }
   n11 = "Transform (Modify) Bone - Bone: dyn_skirt_back_l_0_base"!(ComponentPose=@n10, bAlphaBoolEnabled:bool=True, Translation="0,0,0", Rotation="30.300,123.935,12.679", Scale="1,1,1") -> (Pose) {
    alphaInputType=Bool
    rotationMode=BMM_Additive
    rotationSpace=BCS_ParentBoneSpace
    scaleMode=BMM_Additive
    translationMode=BMM_Additive
    translationSpace=BCS_ParentBoneSpace
    alphaBoolBlend {blendOption=Linear, blendOutTime=5}
    boneToModify {boneName=dyn_skirt_back_l_0_base}
    rotation=<rot 30.3 123.935 12.679>
   }
   n12 = "Transform (Modify) Bone - Bone: dyn_skirt_front_r_0_base"!(ComponentPose=@n11, bAlphaBoolEnabled:bool=True, Translation="0,0,0", Rotation="-24.101,6.035,-33.522", Scale="1,1,1") -> (Pose) {
    alphaInputType=Bool
    rotationMode=BMM_Additive
    rotationSpace=BCS_ParentBoneSpace
    scaleMode=BMM_Additive
    translationMode=BMM_Additive
    translationSpace=BCS_ParentBoneSpace
    alphaBoolBlend {blendOption=Linear, blendOutTime=5}
    boneToModify {boneName=dyn_skirt_front_r_0_base}
    rotation=<rot -24.101 6.035 -33.522>
   }
   n13 = "Transform (Modify) Bone - Bone: dyn_skirt_r_0_base"!(ComponentPose=@n12, bAlphaBoolEnabled:bool=True, Translation="0,0,0", Rotation="-41.118,102.160,-82.898", Scale="1,1,1") -> (Pose) {
    alphaInputType=Bool
    rotationMode=BMM_Additive
    rotationSpace=BCS_ParentBoneSpace
    scaleMode=BMM_Additive
    translationMode=BMM_Additive
    translationSpace=BCS_ParentBoneSpace
    alphaBoolBlend {blendOption=Linear, blendOutTime=5}
    boneToModify {boneName=dyn_skirt_r_0_base}
    rotation=<rot -41.118 102.16 -82.898>
   }
   n14 = "Transform (Modify) Bone - Bone: dyn_skirt_back_r_0_base"!(ComponentPose=@n13, bAlphaBoolEnabled:bool=True, Translation="0,0,0", Rotation="-2.022,114.086,-10.271", Scale="1,1,1") -> (Pose) {
    alphaInputType=Bool
    rotationMode=BMM_Additive
    rotationSpace=BCS_ParentBoneSpace
    scaleMode=BMM_Additive
    translationMode=BMM_Additive
    translationSpace=BCS_ParentBoneSpace
    alphaBoolBlend {blendOption=Linear, blendOutTime=5}
    boneToModify {boneName=dyn_skirt_back_r_0_base}
    rotation=<rot -2.022 114.086 -10.271>
   }
   n15 = "Transform (Modify) Bone - Bone: dyn_skirt_front_l_0_base"(ComponentPose=@n14, bAlphaBoolEnabled=True, Translation="0,0,0", Rotation="-47.424,-118.968,61.380", Scale="1,1,1") {
    alphaInputType=Bool
    rotationMode=BMM_Replace
    rotationSpace=BCS_ParentBoneSpace
    scaleMode=BMM_Additive
    scaleSpace=BCS_ParentBoneSpace
    translationMode=BMM_Additive
    translationSpace=BCS_ParentBoneSpace
    alphaBoolBlend {blendOption=Linear, blendOutTime=5}
    boneToModify {boneName=dyn_skirt_front_l_0_base}
    rotation=<rot -47.424 -118.968 61.38>
   }
   n16 = "Transform (Modify) Bone - Bone: dyn_skirt_l_0_base"(ComponentPose=@n15, bAlphaBoolEnabled=True, Translation="0,0,0", Rotation="-49.593,-87.926,138.348", Scale="1,1,1") {
    alphaInputType=Bool
    rotationMode=BMM_Replace
    rotationSpace=BCS_ParentBoneSpace
    scaleMode=BMM_Additive
    scaleSpace=BCS_ParentBoneSpace
    translationMode=BMM_Additive
    translationSpace=BCS_ParentBoneSpace
    alphaBoolBlend {blendOption=Linear, blendOutTime=5}
    boneToModify {boneName=dyn_skirt_l_0_base}
    rotation=<rot -49.593 -87.926 138.348>
   }
   n17 = "Transform (Modify) Bone - Bone: dyn_skirt_back_l_0_base"(ComponentPose=@n16, bAlphaBoolEnabled=True, Translation="0,0,0", Rotation="-38.797,-73.241,-176.549", Scale="1,1,1") {
    alphaInputType=Bool
    rotationMode=BMM_Replace
    rotationSpace=BCS_ParentBoneSpace
    scaleMode=BMM_Additive
    scaleSpace=BCS_ParentBoneSpace
    translationMode=BMM_Additive
    translationSpace=BCS_ParentBoneSpace
    alphaBoolBlend {blendOption=Linear, blendOutTime=5}
    boneToModify {boneName=dyn_skirt_back_l_0_base}
    rotation=<rot -38.797 -73.241 -176.549>
   }
   n18 = "Transform (Modify) Bone - Bone: dyn_skirt_front_r_0_base"(ComponentPose=@n17, bAlphaBoolEnabled=True, Translation="0,0,0", Rotation="-25.600,143.892,0.985", Scale="1,1,1") {
    alphaInputType=Bool
    rotationMode=BMM_Replace
    rotationSpace=BCS_ParentBoneSpace
    scaleMode=BMM_Additive
    scaleSpace=BCS_ParentBoneSpace
    translationMode=BMM_Additive
    translationSpace=BCS_ParentBoneSpace
    alphaBoolBlend {blendOption=Linear, blendOutTime=5}
    boneToModify {boneName=dyn_skirt_front_r_0_base}
    rotation=<rot -25.6 143.892 0.985>
   }
   n19 = "Transform (Modify) Bone - Bone: dyn_skirt_r_0_base"(ComponentPose=@n18, bAlphaBoolEnabled=True, Translation="0,0,0", Rotation="7.632,135.959,69.367", Scale="1,1,1") {
    alphaInputType=Bool
    rotationMode=BMM_Replace
    rotationSpace=BCS_ParentBoneSpace
    scaleMode=BMM_Additive
    scaleSpace=BCS_ParentBoneSpace
    translationMode=BMM_Additive
    translationSpace=BCS_ParentBoneSpace
    alphaBoolBlend {blendOption=Linear, blendOutTime=5}
    boneToModify {boneName=dyn_skirt_r_0_base}
    rotation=<rot 7.632 135.959 69.367>
   }
   n20 = "Transform (Modify) Bone - Bone: dyn_skirt_back_r_0_base"(ComponentPose=@n19, bAlphaBoolEnabled=True, Translation="0,0,0", Rotation="21.957,123.248,143.305", Scale="1,1,1") {
    alphaInputType=Bool
    rotationMode=BMM_Replace
    rotationSpace=BCS_ParentBoneSpace
    scaleMode=BMM_Additive
    scaleSpace=BCS_ParentBoneSpace
    translationMode=BMM_Additive
    translationSpace=BCS_ParentBoneSpace
    alphaBoolBlend {blendOption=Linear, blendOutTime=5}
    boneToModify {boneName=dyn_skirt_back_r_0_base}
    rotation=<rot 21.957 123.248 143.305>
   }
   n21 = RigidBody(ComponentPose=@n20, OverrideWorldGravity=Get Gravity Override_0() -> (Gravity Override_0:Vector)) {
    bClampLinearTranslationLimitToRefPose=true
    bOverrideWorldGravity=true
    overridePhysicsAsset=&/F_MED_NitroFlow_Physics_slide
    simulationSpace=BaseBoneSpace
    alphaBoolBlend {blendOption=Linear}
    baseBoneRef {boneName=pelvis}
    componentAppliedLinearAccClamp=<1000 1000 1000>
    componentLinearAccScale=<1 1 1>
    componentLinearVelScale=<1 1 1>
    overrideWorldGravity {y=-1000, z=1000}
    simSpaceSettings {maxAngularAcceleration=0, maxAngularVelocity=0, maxLinearAcceleration=100, maxLinearVelocity=100, velocityScaleZ=0, worldAlpha=1}
   }
   n22 = Component To Local(ComponentPose=@n21)
   n23 = Get bIsSliding()
   n24 = Anim Dynamics - Bone: dyn_lace_legs_l_1!(
    ComponentPose=Local To Component(LocalPose=Blend Poses by bool(BlendPose_0=@n22, BlendPose_1=@n8, BlendTime_0=0.3, BlendTime_1=0.3, bActiveValue=@n23) {blendType=Linear})
    Alpha=1
   ) -> (Pose) {
    angularDampingOverride=1
    angularSpringConstant=100
    bOverrideAngularDamping=true
    bUseGravityOverride=true
    simulationSpace=RootRelative
    alphaBoolBlend {blendOption=Linear}
    boundBone {boneName=dyn_lace_legs_l_1}
    componentAppliedLinearAccClamp=<1000 1000 1000>
    componentLinearAccScale=<0.05 0.05 0.05>
    componentLinearVelScale=<0.05 0.05 0.05>
    physicsBodyDefinitions [{
      boundBone {boneName=dyn_lace_legs_l_1}
      boxExtents=<4 4 4>
      constraintSetup {angularLimitsMax=<5 15 15>, angularLimitsMin=<-5 -15 -15>, angularTarget=<1 0 0>}
      localJointOffset=<1 0 0>
     }]
    relativeSpaceBone {boneName=pelvis}
   }
   n25 = Anim Dynamics - Bone: dyn_lace_legs_l_2!(ComponentPose=@n24, Alpha=1) -> (Pose) {
    angularDampingOverride=1
    angularSpringConstant=100
    bOverrideAngularDamping=true
    bUseGravityOverride=true
    simulationSpace=RootRelative
    alphaBoolBlend {blendOption=Linear}
    boundBone {boneName=dyn_lace_legs_l_2}
    componentAppliedLinearAccClamp=<1000 1000 1000>
    componentLinearAccScale=<0.05 0.05 0.05>
    componentLinearVelScale=<0.05 0.05 0.05>
    physicsBodyDefinitions [{
      boundBone {boneName=dyn_lace_legs_l_2}
      boxExtents=<4 4 4>
      constraintSetup {angularLimitsMax=<5 15 15>, angularLimitsMin=<-5 -15 -15>, angularTarget=<1 0 0>}
      localJointOffset=<1 0 0>
     }]
    relativeSpaceBone {boneName=pelvis}
   }
   n26 = Anim Dynamics - Bone: dyn_race_regs_r_1!(ComponentPose=@n25, Alpha=1) -> (Pose) {
    angularDampingOverride=1
    angularSpringConstant=100
    bOverrideAngularDamping=true
    bUseGravityOverride=true
    simulationSpace=RootRelative
    alphaBoolBlend {blendOption=Linear}
    boundBone {boneName=dyn_race_regs_r_1}
    componentAppliedLinearAccClamp=<1000 1000 1000>
    componentLinearAccScale=<0.05 0.05 0.05>
    componentLinearVelScale=<0.05 0.05 0.05>
    physicsBodyDefinitions [{
      boundBone {boneName=dyn_race_regs_r_1}
      boxExtents=<4 4 4>
      constraintSetup {angularLimitsMax=<5 15 15>, angularLimitsMin=<-5 -15 -15>, angularTarget=<1 0 0>}
      localJointOffset=<1 0 0>
     }]
    relativeSpaceBone {boneName=pelvis}
   }
   n27 = Anim Dynamics - Bone: dyn_race_regs_r_2!(ComponentPose=@n26, Alpha=1) -> (Pose) {
    angularDampingOverride=1
    angularSpringConstant=100
    bOverrideAngularDamping=true
    bUseGravityOverride=true
    simulationSpace=RootRelative
    alphaBoolBlend {blendOption=Linear}
    boundBone {boneName=dyn_race_regs_r_2}
    componentAppliedLinearAccClamp=<1000 1000 1000>
    componentLinearAccScale=<0.05 0.05 0.05>
    componentLinearVelScale=<0.05 0.05 0.05>
    physicsBodyDefinitions [{
      boundBone {boneName=dyn_race_regs_r_2}
      boxExtents=<4 4 4>
      constraintSetup {angularLimitsMax=<5 15 15>, angularLimitsMin=<-5 -15 -15>, angularTarget=<1 0 0>}
      localJointOffset=<1 0 0>
     }]
    relativeSpaceBone {boneName=pelvis}
   }
   n28 = Get Built-in Emote() -> (Built-in Emote:double)
   n29 = "Transform (Modify) Bone - Bone: dyn_skirt_r_0"(ComponentPose=@n27, Alpha=@n28, Translation="0,0,0", Rotation="20,0,0", Scale="1,1,1") {
    rotationMode=BMM_Additive
    rotationSpace=BCS_BoneSpace
    alphaBoolBlend {blendOption=Linear}
    boneToModify {boneName=dyn_skirt_r_0}
    rotation {pitch=20}
   }
   n30 = "Transform (Modify) Bone - Bone: dyn_skirt_r_1"(ComponentPose=@n29, Alpha=@n28, Translation="0,0,0", Rotation="-15,0,0", Scale="1,1,1") {
    rotationMode=BMM_Additive
    rotationSpace=BCS_BoneSpace
    alphaBoolBlend {blendOption=Linear}
    boneToModify {boneName=dyn_skirt_r_1}
    rotation {pitch=-15}
   }
   Output Pose(Result=Blend Poses by bool(
     BlendPose_0=Component To Local(ComponentPose=@n30)
     BlendPose_1=Use cached pose 'befor_skirt'()
     bActiveValue=Get skirt_on/off() -> (skirt_on/off:bool)
    ) {blendType=Linear}) {layerGroup=None}
   n31 = Control Rig!(Source=Copy Pose From Mesh() {bUseAttachedParent=true}, Alpha, constraintWeight:float=Get Constraint Weight() -> (Constraint Weight:float)) -> (Pose) {
    alphaBoolBlend {blendOption=Linear}
    controlRigAssetReference {blueprintRigClass=&/F_MED_NitroFlow_ControlRig_new.F_MED_NitroFlow_ControlRig_new_C}
   }
   n32 = Anim Dynamics - Bone: dyn_lace_l_1!(ComponentPose=Local To Component(LocalPose=@n31), Alpha=1) -> (Pose) {
    angularDampingOverride=1
    angularSpringConstant=100
    bAngularSpring=true
    bGravityOverrideInSimSpace=true
    bOverrideAngularDamping=true
    bUseGravityOverride=true
    simulationSpace=BoneRelative
    alphaBoolBlend {blendOption=Linear}
    boundBone {boneName=dyn_lace_l_1}
    componentAppliedLinearAccClamp=<1000 1000 1000>
    componentLinearAccScale=<0.1 0.1 0.1>
    componentLinearVelScale=<0.1 0.1 0.1>
    physicsBodyDefinitions [{
      boundBone {boneName=dyn_lace_l_1}
      boxExtents=<2 2 2>
      constraintSetup {angularLimitsMax=<10 20 20>, angularLimitsMin=<-10 -20 -20>, angularTarget=<1 0 0>}
      localJointOffset=<2 0 0>
     }]
    relativeSpaceBone {boneName=spine_05}
   }
   n33 = Anim Dynamics - Bone: dyn_lace_l_2!(ComponentPose=@n32, Alpha=1) -> (Pose) {
    angularDampingOverride=1
    angularSpringConstant=100
    bAngularSpring=true
    bGravityOverrideInSimSpace=true
    bOverrideAngularDamping=true
    bUseGravityOverride=true
    simulationSpace=BoneRelative
    alphaBoolBlend {blendOption=Linear}
    boundBone {boneName=dyn_lace_l_2}
    componentAppliedLinearAccClamp=<1000 1000 1000>
    componentLinearAccScale=<0.1 0.1 0.1>
    componentLinearVelScale=<0.1 0.1 0.1>
    physicsBodyDefinitions [{
      boundBone {boneName=dyn_lace_l_2}
      boxExtents=<2 2 2>
      constraintSetup {angularLimitsMax=<10 20 20>, angularLimitsMin=<-10 -20 -20>, angularTarget=<1 0 0>}
      localJointOffset=<2 0 0>
     }]
    relativeSpaceBone {boneName=spine_05}
   }
   n34 = Anim Dynamics - Bone: dyn_lace_l_3!(ComponentPose=@n33, Alpha=1) -> (Pose) {
    angularDampingOverride=1
    angularSpringConstant=100
    bAngularSpring=true
    bGravityOverrideInSimSpace=true
    bOverrideAngularDamping=true
    bUseGravityOverride=true
    simulationSpace=BoneRelative
    alphaBoolBlend {blendOption=Linear}
    boundBone {boneName=dyn_lace_l_3}
    componentAppliedLinearAccClamp=<1000 1000 1000>
    componentLinearAccScale=<0.1 0.1 0.1>
    componentLinearVelScale=<0.1 0.1 0.1>
    physicsBodyDefinitions [{
      boundBone {boneName=dyn_lace_l_3}
      boxExtents=<2 2 2>
      constraintSetup {angularLimitsMax=<10 20 20>, angularLimitsMin=<-10 -20 -20>, angularTarget=<1 0 0>}
      localJointOffset=<2 0 0>
     }]
    relativeSpaceBone {boneName=spine_05}
   }
   n35 = Anim Dynamics - Bone: dyn_race_r_1!(ComponentPose=@n34, Alpha=1) -> (Pose) {
    angularDampingOverride=1
    angularSpringConstant=100
    bAngularSpring=true
    bGravityOverrideInSimSpace=true
    bOverrideAngularDamping=true
    bUseGravityOverride=true
    simulationSpace=BoneRelative
    alphaBoolBlend {blendOption=Linear}
    boundBone {boneName=dyn_race_r_1}
    componentAppliedLinearAccClamp=<1000 1000 1000>
    componentLinearAccScale=<0.1 0.1 0.1>
    componentLinearVelScale=<0.1 0.1 0.1>
    physicsBodyDefinitions [{
      boundBone {boneName=dyn_race_r_1}
      boxExtents=<2 2 2>
      constraintSetup {angularLimitsMax=<10 20 20>, angularLimitsMin=<-10 -20 -20>, angularTarget=<1 0 0>}
      localJointOffset=<2 0 0>
     }]
    relativeSpaceBone {boneName=spine_05}
   }
   n36 = Anim Dynamics - Bone: dyn_race_r_2!(ComponentPose=@n35, Alpha=1) -> (Pose) {
    angularDampingOverride=1
    angularSpringConstant=100
    bAngularSpring=true
    bGravityOverrideInSimSpace=true
    bOverrideAngularDamping=true
    bUseGravityOverride=true
    simulationSpace=BoneRelative
    alphaBoolBlend {blendOption=Linear}
    boundBone {boneName=dyn_race_r_2}
    componentAppliedLinearAccClamp=<1000 1000 1000>
    componentLinearAccScale=<0.1 0.1 0.1>
    componentLinearVelScale=<0.1 0.1 0.1>
    physicsBodyDefinitions [{
      boundBone {boneName=dyn_race_r_2}
      boxExtents=<2 2 2>
      constraintSetup {angularLimitsMax=<10 20 20>, angularLimitsMin=<-10 -20 -20>, angularTarget=<1 0 0>}
      localJointOffset=<2 0 0>
     }]
    relativeSpaceBone {boneName=spine_05}
   }
   n37 = Anim Dynamics - Bone: dyn_race_r_3!(ComponentPose=@n36, Alpha=1) -> (Pose) {
    angularDampingOverride=1
    angularSpringConstant=100
    bAngularSpring=true
    bGravityOverrideInSimSpace=true
    bOverrideAngularDamping=true
    bUseGravityOverride=true
    simulationSpace=BoneRelative
    alphaBoolBlend {blendOption=Linear}
    boundBone {boneName=dyn_race_r_3}
    componentAppliedLinearAccClamp=<1000 1000 1000>
    componentLinearAccScale=<0.1 0.1 0.1>
    componentLinearVelScale=<0.1 0.1 0.1>
    physicsBodyDefinitions [{
      boundBone {boneName=dyn_race_r_3}
      boxExtents=<2 2 2>
      constraintSetup {angularLimitsMax=<10 20 20>, angularLimitsMin=<-10 -20 -20>, angularTarget=<1 0 0>}
      localJointOffset=<2 0 0>
     }]
    relativeSpaceBone {boneName=spine_05}
   }
   Save cached pose 'befor_skirt'(Pose=Component To Local(ComponentPose=@n37)) {cacheName=befor_skirt}
  ]}
 EventGraph {nodes [
   Event BlueprintUpdateAnimation()
   > n1 = Sequence()
   > n2 = Branch(Condition=Get bIsSliding())
   > n3 = Gravity Override State Function DEPRECATED(
    CharacterPartAnimInstance=Self-Reference()
    GameTimeInSeconds=GetGameTimeInSeconds()
    GravityOverrideParams=Get Gravity Override Params_0() -> (Gravity Override Params_0:Gravity Override Params Struct Structure)
   )
   > Set Gravity Override_0(Gravity Override_0:Vector=@n3) -> (Output_Get:Vector)
   n4 = Gravity Override State Function DEPRECATED(
    execute=@n2.else
    CharacterPartAnimInstance=Self-Reference()
    GameTimeInSeconds=GetGameTimeInSeconds()
    GravityOverrideParams=Get Gravity Override Params()
   )
   > Set Gravity Override(Gravity Override=@n4)
   > Set thigh_l_rotation_alpha(thigh_l_rotation_alpha:double=GetCurveValue(CurveName=thigh_l_sherical_reader_value)) -> (Output_Get:double)
   > Set thigh_r_rotation_alpha(thigh_r_rotation_alpha:double=GetCurveValue(CurveName=thigh_r_sherical_reader_value)) -> (Output_Get:double)
   n5 = Branch(execute=@n1.then_1, Condition=NOT Boolean(A=Get bIsInFrontEnd()))
   > Set Built-in Emote(Built-in Emote:double=GetCurveValue(
     self=GetOwnerFortPlayerAnimInstance!(self:Custom Character Part Anim Instance Object Reference) -> (ReturnValue:Fort Player Anim Instance Object Reference)
     CurveName=emote_BuiltIn
    )) -> (Output_Get:double)
   n6 = Set FrontendCheck(execute=@n5.else, FrontendCheck:double=GetCurveValue(self=GetOwnerFrontEndAnimInstance(), CurveName=emote_BuiltIn)) -> (Output_Get:double)
   > Set Built-in Emote(Built-in Emote=@n6)
   n7 = GetOwnerFrontEndAnimInstance()
   > n8 = Branch(Condition="float > float"(A=GetCurveValue(self=@n7, CurveName=FrontEnd_Idle_Bypass), B=0.1))
   > Set Constraint Weight(Constraint Weight:float=0.5) -> (Output_Get:float)
   Set Constraint Weight(execute=@n8.else, Constraint Weight=1)
   Event OnCharacterCustomizationLoaded()
   > Set skirt_on/off(skirt_on/off:bool=CharacterItemDefinitionHasAnyMetaTag(Tags=Make Array("[0]"=(TagName="Cosmetics.Emote.CharacterGroup.Skirts")))) -> (Output_Get:bool)
  ]}
]
variables [
 Built-in Emote {type=number}
 Constraint Weight {type=number, default=1}
 FrontendCheck {type=number}
 Gravity Override {type=Vector}
 Gravity Override Params {
  type=GravityOverrideParamsStruct
  default {
   dBNO {gravityOverride=<-100 100 0>}
   dBNO_Carried {gravityOverride=<-100 100 -100>}
   emoteOrMelee {gravityOverride=<0 0 -100>}
   front_End {gravityOverride=<0 0 -500>}
   onGround_Falling {gravityOverride=<0 0 -100>}
   onGround_Moving {gravityOverride=<0 0 -100>}
   onGround_Standing {gravityOverride=<0 0 -100>}
   riding_Moving {gravityOverride=<0 -100 -100>}
   riding_Moving_Targeting {gravityOverride=<0 -200 -100>}
   riding_Standing {gravityOverride=<0 -100 -100>}
   skydive_DivingDown {gravityOverride=<-200 0 0>}
   skydive_FloatingDown {jointName=pelvis, gravityOverride=<0 0 0>, windAmplitude=<0 0 0>, windFrequency=<0 0 1>, windOffset=<0 0 0>}
   skydive_GoingUp {jointName=pelvis, gravityOverride=<0 0 0>, windAmplitude=<0 0 0>, windFrequency=<0 0 1>}
   skydive_Parachuting {gravityOverride=<0 -100 -100>}
   skydive_Surfing {gravityOverride=<100 -100 -100>}
   tactical_Sprint {gravityOverride=<0 -100 -100>}
   ziplining {gravityOverride=<0 0 -100>}
  }
 }
 Gravity Override Params_0 {
  type=GravityOverrideParamsStruct
  default {onGround_Moving {gravityOverride=<0 -1000 1000>, windAmplitude=<2000 2000 2000>, windFrequency=<500 500 500>}}
 }
 Gravity Override_0 {type=Vector}
 skirt_on/off {type=boolean}
 thigh_l_rotation_alpha {type=number}
 thigh_r_rotation_alpha {type=number}
]
